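# Master Level Sequence for PistolCanary_EditorOnly, a timeline just over one second long that assembles subsequences and has no bindings of its own.
# Event and Editor Only sub-tracks span essentially the whole range.
# Section 01 plays near the start and Section 03 covers the final part, leaving a gap between them with no Section 02 track.
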LevelSequence /PistolCanary_EditorOnly/Sequences/PistolCanary_EditorOnly_Main
bindings []
display_rate {denominator=1, numerator=60}
tick_resolution {denominator=1, numerator=24000}
tracks [
 {type=MovieSceneSubTrack, display_name=PistolCanary_Event, sections [{type=MovieSceneSubSection, end_frame=26784, start_frame=0}]}
 {
  type=MovieSceneSubTrack
  display_name=PistolCanary_EditorOnly - Section 01
  sections [{type=MovieSceneSubSection, end_frame=7850, start_frame=1728}]
 }
 {
  type=MovieSceneSubTrack
  display_name=PistolCanary_EditorOnly - Section 03
  sections [{type=MovieSceneSubSection, end_frame=26783, start_frame=18417}]
 }
 {
  type=MovieSceneSubTrack
  display_name=PistolCanary_EditorOnly - Editor Only
  sections [{type=MovieSceneSubSection, end_frame=26783, start_frame=0}]
 }
]
